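# Additive unlit Niagara sprite material that draws procedural bokeh discs, with dynamic parameters for size, softness, noise and fades.
# Two sphere masks form a soft disc with a dimmer center and brighter rim, scaling particle-colored emissive and alpha.
# Panning cell noise breaks up the alpha, which a near-camera fade attenuates.
# A switchable fake sun-shadow term blends sun and shadow multipliers by view angle to the key light.
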Material /Game/Effects/Fort_Effects/Effects/Environment/Hermes/IcicleLarva/M_FOS_Bokeh
graph {
 properties {
  bAutomaticallySetUsageInEditor=false
  bCastRayTracedShadows=false
  bGenerateSphericalParticleNormals=true
  bUsedWithNiagaraSprites=true
  blendMode=BLEND_Additive
  shadingModel=MSM_Unlit
  translucencyLightingMode=TLM_VolumetricDirectional
  usedShadingModels=Unlit
 }
 nodes [
  DynamicParameter() {parameterIndex=3, defaultValue=<rgba 512 1 1 1>, paramNames [NearFade, FadeDistance, SunSide_Mult, ShadowSide_Mult]}
  n1 = DynamicParameter() {parameterIndex=1, defaultValue=<rgba 0.25 0.25 1 25>, paramNames [NoiseTex_Tile_X, NoiseTex_Tile_Y, NoiseTex_Contrast, Emissive]}
  n2 = ParticleColor()
  n3 = NamedRerouteUsage()
  MaterialFunctionCall(Emissive=Multiply(A=@n1.Emissive, B=Multiply(A=@n2.RGB, B=@n3))) {materialFunction=&MF_FX_EmissiveAdjustments}
  n4 = DynamicParameter() {parameterIndex=3, defaultValue=<rgba 512 1 1 1>, paramNames [NearFade, FadeDistance, SunSide_Mult, ShadowSide_Mult]}
  n5 = DynamicParameter() {parameterIndex=1, defaultValue=<rgba 0.25 0.25 1 25>, paramNames [NoiseTex_Tile_X, NoiseTex_Tile_Y, NoiseTex_Contrast, Emissive]}
  n6 = Multiply(A=TextureCoordinate(), B=AppendVector(A=@n5.NoiseTex_Tile_X, B=@n5.NoiseTex_Tile_X))
  n7 = DynamicParameter() {parameterIndex=2, defaultValue=<rgba 512 250 0.02 0.02>, paramNames [CameraDepth_Fade, CameraDepth_FadeOffset, PannerSpeed_X, PannerSpeed_Y]}
  n8 = AppendVector(A=@n7.PannerSpeed_X, B=@n7.PannerSpeed_X)
  n9 = MaterialFunctionCall(
   In=TextureSample(UVs=Panner(Coordinate=@n6, Speed=@n8)) {texture=/Game/Global/Textures/Default/Noise/T_SoftCells_Negative}.G
   Contrast=@n5.NoiseTex_Contrast
  ) {materialFunction=&CheapContrast}
  n10 = Multiply(A=@n3, B=Multiply(A=ParticleColor().A, B=LinearInterpolate(Alpha=@n9.Result) {constA=1, constB=0.1}))
  Multiply(
   A=Multiply(
    A=MaterialFunctionCall(Distance=@n4.NearFade) {materialFunction=&Near_Camera_Fade}.Result
    B=@n10
   )
   B=StaticSwitchParameter(True=NamedRerouteUsage(), False=Constant() {r=1}) {defaultValue=true, parameterName=UseFakeSunShadowing?}
  )
  n11 = DynamicParameter() {parameterIndex=2, defaultValue=<rgba 512 250 0.02 0.02>, paramNames [CameraDepth_Fade, CameraDepth_FadeOffset, PannerSpeed_X, PannerSpeed_Y]}
  Multiply(
   A=@n10
   B=MaterialFunctionCall(Fade Length=@n11.CameraDepth_Fade, Fade Offset=@n11.CameraDepth_Fade) {materialFunction=&CameraDepthFade}.Result
  )
  n12 = TextureCoordinate()
  n13 = DynamicParameter() {defaultValue=<rgba 0.48 0.8 0.5 0.5>, paramNames [Radius, Hardness, Sphere_Offset_X, Sphere_Offset_Y]}
  n14 = AppendVector(A=@n13.Sphere_Offset_X, B=@n13.Sphere_Offset_X)
  n15 = DynamicParameter() {defaultValue=<rgba 0.48 0.8 0.5 0.5>, paramNames [Radius, Hardness, Sphere_Offset_X, Sphere_Offset_Y]}
  n16 = Clamp(None=Subtract(A=@n15.Radius) {constB=0.02})
  n17 = ScalarParameter() {defaultValue=0.94, parameterName=InnerRadius}
  n18 = OneMinus(None=SphereMask(A=TextureCoordinate(), B=@n14, Radius=Multiply(A=@n16, B=@n17), Hardness=Constant() {r=0.95}))
  NamedRerouteDeclaration(None=Multiply(A=SphereMask(A=@n12, B=@n14, Radius=@n16, Hardness=@n15.Hardness), B=LinearInterpolate(Alpha=@n18) {constA=0.2})) {name=Name_SphereMask, nodeColor=<rgba 1 0.7734 0 1>}
  n19 = DynamicParameter() {parameterIndex=3, defaultValue=<rgba 512 1 1 1>, paramNames [NearFade, FadeDistance, SunSide_Mult, ShadowSide_Mult]}
  n20 = ScalarParameter() {parameterName=ClampKeyLightFade}
  n21 = Clamp(None=@n19.SunSide_Mult, Min=@n20) {maxDefault=5}
  n22 = Clamp(None=@n19.ShadowSide_Mult, Min=@n20) {maxDefault=5}
  n23 = VectorParameter() {parameterName=KeyLightDirection, defaultValue=<rgba 0 1 1 1>}
  n24 = Clamp(None=DotProduct(A=ComponentMask(None=@n23.RGB) {b=true, g=true, r=true}, B=CameraVectorWS()))
  NamedRerouteDeclaration(None=Power(Base=LinearInterpolate(A=@n21, B=@n22, Alpha=@n24), Exp=ScalarParameter() {defaultValue=5, parameterName=FakeSunShadowingPower})) {name=FakeSunShadowing, nodeColor=<rgba 1 0.3866 0 1>}
  n25 = DynamicParameter() {parameterIndex=3, defaultValue=<rgba 512 1 1 1>, paramNames [NearFade, FadeDistance, SunSide_Mult, ShadowSide_Mult]}
  n26 = Divide(A=PixelDepth().R, B=@n25.NearFade) {constB=512}
  SmoothStep(Value=Saturate(None=Subtract(A=@n26) {constB=2}))
 ]
}
