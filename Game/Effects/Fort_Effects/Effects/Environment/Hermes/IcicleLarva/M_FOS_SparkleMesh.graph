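# Additive material for sparkle meshes that draws camera-facing, looping flare sprites using world position offset.
# Vertex color red varies phase and lifetime for a normalized age, green varies sprite scale, and blue drives an opacity gradient.
# Opacity combines the flare texture with age fade, near/far distance fade, that gradient and time-of-day scalars.
# Tinted flare color feeds HDR emissive adjustments, and lower quality tightens the distance-fade range.
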Material /Game/Effects/Fort_Effects/Effects/Environment/Hermes/IcicleLarva/M_FOS_SparkleMesh
graph {
 properties {bAutomaticallySetUsageInEditor=false, blendMode=BLEND_Additive, usedShadingModels=Default Lit}
 nodes [
  n1 = TextureSampleParameter2D(UVs=TextureCoordinate()) {parameterName=SparkleTexture, texture=/Game/Effects/Fort_Effects/Textures/Flares/T_Anime_Flare}
  n2 = Multiply(A=Multiply(A=VectorParameter() {group=Color, parameterName=Color, sortPriority=0, defaultValue=<rgba 1 1 1 1>}.RGB, B=@n1.G))
  MaterialFunctionCall(Emissive=MaterialFunctionCall(LDR Input=Clamp(None=@n2), HDR Tint=@n2) {materialFunction=&DeriveHDRfromLDR}.RemappedRange) {materialFunction=&MF_FX_EmissiveAdjustments}
  n3 = Add(
   A=Multiply(A=VertexColor().R) {constB=100}
   B=Multiply(A=Time(), B=ScalarParameter() {defaultValue=1, group=Time, parameterName=Delta Time Scale, sortPriority=0})
  )
  n4 = LinearInterpolate(
   A=ScalarParameter() {defaultValue=0.5, group=Time, parameterName=Min Lifetime, sortPriority=1}
   B=ScalarParameter() {defaultValue=1, group=Time, parameterName=Max Lifetime, sortPriority=2}
   Alpha=VertexColor().R
  )
  NamedRerouteDeclaration(None=Divide(A=Fmod(A=@n3, B=@n4), B=@n4)) {name=NormalizedAge, nodeColor=<rgba 0.7812 1 0 1>}
  n5 = TextureCoordinate()
  n6 = LinearInterpolate(
   A=ScalarParameter() {defaultValue=0.5, group=WPO, parameterName=Min Uniform Scale, sortPriority=1}
   B=ScalarParameter() {defaultValue=0.5, group=WPO, parameterName=Max Uniform Scale, sortPriority=2}
   Alpha=VertexColor().G
  )
  n7 = Multiply(
   A=Multiply(A=AppendVector(A=@n6, B=@n6), B=ScalarParameter() {defaultValue=1, group=WPO, parameterName=Uniform Scale, sortPriority=3})
   B=NamedRerouteUsage()
  )
  n8 = WorldPosition()
  n9 = ScalarParameter() {group=WPO, parameterName=Camera Offset, sortPriority=0}
  NamedRerouteDeclaration(None=Add(
    A=MaterialFunctionCall!(Center Location=@n8.XYZ, 0-1 UVs=@n5, XY scale=@n7, OverridePivot, "Normals (Optional)") {materialFunction=&Sprite}.World position offset
    B=MaterialFunctionCall(Offset Amount=@n9, World Position=@n8.XYZ) {materialFunction=/Engine/Functions/Engine_MaterialFunctions02/WorldPositionOffset/CameraOffset}.Camera Offset
   )) {name=WPO, nodeColor=<rgba 0 0.0625 1 1>}
  NamedRerouteDeclaration(None=SphereMask(
    A=NamedRerouteUsage()
    B=ScalarParameter() {defaultValue=0.5, group=Opacity, parameterName=OpacityFadeStart, sortPriority=0}
    Radius=ScalarParameter() {defaultValue=0.5, group=Opacity, parameterName=OpacityFadeDuration, sortPriority=1}
    Hardness=ScalarParameter() {group=Opacity, parameterName=OpacityFadeHardness, sortPriority=2}
   )) {name=AgeOpacityRamp, nodeColor=<rgba 1 0 0.8515 1>}
  n10 = ScalarParameter() {defaultValue=21000, group=CloseFarFader, parameterName="NearFadeLength (S)"}
  n11 = ShadingPathSwitch(
   Default=ScalarParameter() {defaultValue=0.65, group=CloseFarFader, parameterName=DistanceScaleMult_LowQuality}
   Deferred=QualitySwitch(Default=Constant() {r=1}, Low=ScalarParameter() {defaultValue=0.65, group=CloseFarFader, parameterName=DistanceScaleMult_LowQuality})
  ) {desc=Squish WPO earlier on lower quality}
  n12 = ScalarParameter() {defaultValue=20000, group=CloseFarFader, parameterName="NearFadeOffset (S)"}
  n13 = ScalarParameter() {defaultValue=25000, group=CloseFarFader, parameterName="FarFadeLength (S)"}
  n14 = ScalarParameter() {defaultValue=125000, group=CloseFarFader, parameterName="FarFadeOffset (S)"}
  NamedRerouteDeclaration(None=MaterialFunctionCall!(
    NearFadeLength=Divide(A=@n10, B=@n11)
    NearFadeOffset=Divide(A=@n12, B=@n11)
    NearFadeCurve
    NearFadeStartingScale
    NearFadeEndingScale
    FarFadeLength=Multiply(A=@n13, B=@n11)
    FarFadeOffset=Multiply(A=@n14, B=@n11)
    FarFadeCurve
    FarFadeEndingScale
    Opacity Near Distance Bias
    Opacity Far Distance Bias=ScalarParameter() {defaultValue=-2, group=CloseFarFader, parameterName="Opacity Far Distance Bias (S)"}
   ) {materialFunction=/Game/Effects/Fort_Effects/Materials/_Functions/MF_SmoothLod}.Opacity) {name=NearFarFade, nodeColor=<rgba 0 1 0.0859 1>}
  n15 = NamedRerouteUsage()
  n16 = NamedRerouteUsage()
  n17 = Multiply(
   A=@n1.G
   B=Multiply(A=Multiply(A=@n15, B=@n16), B=ScalarParameter() {defaultValue=1, group=Opacity, parameterName=OpacityScalar, sortPriority=-1})
  )
  n18 = NamedRerouteUsage()
  NamedRerouteDeclaration(None=Multiply(A=Multiply(A=@n17, B=@n18), B=NamedRerouteUsage())) {name=Opacity, nodeColor=<rgba 1 0 0.7969 1>}
  n19 = VertexColor()
  n20 = ScalarParameter() {defaultValue=1, group=Opacity, parameterName=ZOpacityExponent, sortPriority=7}
  NamedRerouteDeclaration(None=LinearInterpolate(
    A=ScalarParameter() {group=Opacity, parameterName=MinZOpacity, sortPriority=5}
    B=ScalarParameter() {defaultValue=1, group=Opacity, parameterName=MaxZOpacity, sortPriority=6}
    Alpha=Power(Base=@n19.B, Exp=@n20)
   )) {name=ZOpacity, nodeColor=<rgba 1 0.0563 0.9635 1>}
  NamedRerouteDeclaration(None=MaterialFunctionCall(
    Value_Morning=ScalarParameter() {defaultValue=1, group=Opacity TOD, parameterName=Opacity Morning, sortPriority=0}
    Value_Day=ScalarParameter() {defaultValue=1, group=Opacity TOD, parameterName=Opacity Day, sortPriority=1}
    Value_Evening=ScalarParameter() {defaultValue=1, group=Opacity TOD, parameterName=Opacity Evening, sortPriority=2}
    Value_Night=ScalarParameter() {defaultValue=1, group=Opacity TOD, parameterName=Opacity Night, sortPriority=3}
   ) {materialFunction=&Scalar_Time_of_Day}.Result) {name=OpacityTOD, nodeColor=<rgba 1 0.0031 0.706 1>}
  NamedRerouteUsage()
  NamedRerouteUsage()
 ]
}
