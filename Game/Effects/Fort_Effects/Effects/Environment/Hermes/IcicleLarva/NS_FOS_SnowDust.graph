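# Niagara environment effect that fills the air with drifting, twinkling snow dust, using a single GPU sprite emitter.
# Three active spawn groups feed offset sphere volumes, with system parameters setting the rate multiplier and color range.
# Weak downward gravity, drag and panning curl noise move the motes.
# Camera distance drives sprite size, while waveforms, a lifetime curve and a region mask modulate brightness and alpha.
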NiagaraSystem /Game/Effects/Fort_Effects/Effects/Environment/Hermes/IcicleLarva/NS_FOS_SnowDust
dependencies {
 usedDataInterfaces [&NiagaraDataInterfaceCamera, &NiagaraDataInterfaceCurve]
 usedDynamicInputs []
 usedModules [
  &SystemState
  &self:SystemSpawnScript.SystemScriptSource.SystemScriptGraph.NiagaraNodeAssignment_0.SetVariables_#1
  &self:SystemSpawnScript.SystemScriptSource.SystemScriptGraph.NiagaraNodeAssignment_1.SetVariables_#2
 ]
}
emitters {Dust_Motes001 {
  simTarget=GPUComputeSim
  emitterUpdateScript [
   /SetVariables(Emitter.CameraPostion:Vector3f=&ReturnCameraVector(Camera Property=::NewEnumerator0, Coordinate Space=::Simulation)) {isSetParameters=true}
   &EmitterState(
    Life Cycle Mode=::NewEnumerator1
    Loop Duration=&RandomRangeFloat(Minimum=5, Maximum=10)
    Recalculate Duration Each Loop=true
    UseLoopDelay=false
   )
   &SpawnRate(SpawnRate=&Multiply_Float(A=2000, B=$System.SpawnrateMult), Use Spawn Probability=false) {enabled=false, module=SpawnRate001}
   &SpawnRate(SpawnGroup=1, SpawnRate=&Multiply_Float(A=400, B=$System.SpawnrateMult), Use Spawn Probability=false) {module=SpawnRate002}
   &SpawnRate(SpawnGroup=2, SpawnRate=&Multiply_Float(A=2000, B=$System.SpawnrateMult), Use Spawn Probability=false) {module=SpawnRate003}
   &SpawnRate(SpawnGroup=3, SpawnRate=&Multiply_Float(A=2000, B=$System.SpawnrateMult), Use Spawn Probability=false) {module=SpawnRate004}
  ]
  particleSpawnScript [
   /SetVariables(Particles.RandSpreadCurve:NiagaraFloat=&FloatFromCurve(
     FloatCurve=DataInterface() {propertyValues {Curve {keys [
         {interpMode=RCIM_Cubic, tangentMode=RCTM_User, time=0, value=0.7}
         {interpMode=RCIM_Cubic, tangentMode=RCTM_User, time=0.8, value=0.4}
         {arriveTangent=2.7919, interpMode=RCIM_Cubic, leaveTangent=2.7919, tangentMode=RCTM_User, time=1, value=1}
        ]}}}
     CurveIndex=&RandomRangeFloat(Minimum=0, Maximum=1)
    )) {isSetParameters=true}
   &InitializeParticle(
    Color Maximum=$System.ColorNear
    Color Minimum=$System.ColorFar
    Color Mode=::NewEnumerator2
    Lifetime=&UniformRangedFloat(Minimum=2, Maximum=3)
    Mass Mode=::NewEnumerator1
    Sprite Rotation Mode=::NewEnumerator2
    Sprite Size=&Vector2DFromFloat(Value=&RandomRangeFloat(Minimum=1, Maximum=5))
    Sprite Size Mode=::NewEnumerator4
    UsePositionOffset=false
    Write Material Random=true
   )
   &ShapeLocation(MaskSpawnGroup=true, Override Seed=false, Rotation Mode=::NewEnumerator1, Sphere Radius=1500, UseOwnerScale=true) {enabled=false}
   &ShapeLocation(
    MaskSpawnGroup=true
    Offset=<0 -2000 -400>
    Offset Mode=::NewEnumerator0
    Override Seed=false
    Rotation Mode=::NewEnumerator1
    Spawn Group Mask=1
    Sphere Radius=800
    UseOwnerScale=true
   ) {module=ShapeLocation001}
   &ShapeLocation(
    MaskSpawnGroup=true
    Offset=<2500 0 1000>
    Offset Mode=::NewEnumerator0
    Override Seed=false
    Rotation Mode=::NewEnumerator1
    Spawn Group Mask=2
    Sphere Radius=1500
    UseOwnerScale=true
   ) {module=ShapeLocation002}
   &ShapeLocation(
    MaskSpawnGroup=true
    Offset=<0 800 500>
    Offset Mode=::NewEnumerator0
    Override Seed=false
    Rotation Mode=::NewEnumerator1
    Spawn Group Mask=3
    Sphere Radius=1500
    UseOwnerScale=true
   ) {module=ShapeLocation004}
   &AddVelocity(Velocity=&RandomVector(Vector Scale=&RandomRangeFloat(Minimum=0.5, Maximum=1)))
   /Niagara/Modules/Update/Camera/MaintainInCameraParticleScale()
   &DynamicMaterialParameters(
    Index 0 Param 1=2
    Index 0 Param 2=50
    Index 0 Param 3=200
    Module.Param0WriteEnabled=true
    Module.Param0WriteEnabled3=true
    Module.Param1WriteEnabled=true
    Module.Param1WriteEnabled3=true
    Module.Param2WriteEnabled=true
    Module.Param2WriteEnabled3=true
    Module.Param3WriteEnabled=true
    Module.Param3WriteEnabled3=true
    Write Parameter Index 3=true
   ) {module=DynamicMaterialParameters001}
  ]
  particleUpdateScript [
   &ParticleState()
   &self:Camera_to_Particle_Distance_Calculation(Camera Postion:Vector3f=$Emitter.CameraPostion, Distance=1500, Particle Position:NiagaraPosition=$Particles.Position) {module=Camera_to_Particle_Distance_Calculation}
   &ScaleSpriteSize(Scale Factor=&Vector2DFromFloat(Value=&FloatFromCurve(
      FloatCurve=DataInterface() {propertyValues {Curve {keys [<at 0 0.25 RCIM_Cubic>,<at 1 1.5 RCIM_Cubic>]}}}
      CurveIndex=$Particles.NormalizedDistance
     )))
   &GravityForce(Gravity=&MakeVector(
     X=0
     Y=0
     Z=&FloatFromCurve(
      FloatCurve=DataInterface() {propertyValues {Curve {keys [<at 0 1 RCIM_Cubic>,<at 1 0>]}}}
      Scale Curve=&RandomRangeFloat(Minimum=-15, Maximum=-25)
     )
    ))
   &Drag(Drag=0.5, Use Linear Drag=true, Use Rotational Drag=false)
   &CurlNoiseForce(
    Noise Frequency=&RandomRangeFloat(Minimum=0.2, Maximum=1)
    Noise Strength=&RandomRangeFloat(Minimum=15, Maximum=25)
    Pan Noise=true
    Pan Noise Field=<0 0 1>
   )
   &SpriteRotationRate(Rotation Rate=&RandomRangeFloat(Minimum=-45, Maximum=45))
   &ScaleColor(
    Scale Alpha=&Multiply_Float(
     A=&Lerp_Float(A=1, B=0.3, Alpha=$Particles.RandSpreadCurve)
     B=&Waveform(
      "Waveform [1]"=::NewEnumerator3
      "[1] Frequency"=&RandomRangeFloat(Minimum=0.1, Maximum=0.25)
      "[1] Amplitude Min/Max"=<0.5 1>
     )
    )
    Scale RGB=&VectorFromFloat(Value=&Lerp_Float(A=0.5, B=1, Alpha=$Particles.RandSpreadCurve))
    ScaleA=true
    ScaleRGB=true
   ) {module=ScaleColor001}
   &RegionMask(Region Radius=5000, UseOffset=false, UseRotation=false)
   &ScaleColor(
    Scale Alpha=&FloatFromCurve(
     FloatCurve=DataInterface() {propertyValues {Curve {
        defaultValue=0
        keys [
         {interpMode=RCIM_Cubic, tangentMode=RCTM_User, time=0, value=0}
         {interpMode=RCIM_Cubic, tangentMode=RCTM_User, time=0.5, value=1}
         {tangentMode=RCTM_User, time=1, value=0}
        ]
       }}}
     Scale Curve=&Multiply_Float(A=3, B=$Particles.RegionMask.Float)
    )
    Scale RGB=&Multiply_VectorByFloat(Vector=<1 1 1>, Float=1)
    ScaleA=true
    ScaleRGB=true
   )
   &SolveForcesAndVelocity(Acceleration Limit=5, Clamp Velocity=true, Limit Acceleration=true, Speed Limit=100)
   &DynamicMaterialParameters(
    Index 0 Param 1=0.5
    Index 0 Param 2=1
    Index 0 Param 3=0.5
    Index 0 Param 4=0.5
    Index 1 Param 1=1
    Index 1 Param 2=1
    Index 1 Param 3=1
    Index 1 Param 4=&RandomRangeFloat(Minimum=0, Maximum=0.5)
    Index 2 Param 1=512
    Index 2 Param 2=256
    Index 2 Param 3=1
    Index 2 Param 4=1
    Index 3 Param 1=512
    Index 3 Param 3=1
    Index 3 Param 4=0.5
    Module.Index0Param0Write=true
    Module.Index0Param1Write=true
    Module.Index0Param2Write=true
    Module.Index0Param3Write=true
    Module.Index1Param0Write=true
    Module.Index1Param1Write=true
    Module.Index1Param2Write=true
    Module.Index1Param3Write=true
    Module.Index2Param0Write=true
    Module.Index2Param1Write=true
    Module.Index2Param2Write=true
    Module.Index2Param3Write=true
    Module.Index3Param0Write=true
    Module.Index3Param1Write=true
    Module.Index3Param2Write=true
    Module.Index3Param3Write=true
    Write Parameter Index 1=true
    Write Parameter Index 2=true
    Write Parameter Index 3=true
   )
   /SetVariables(
    Particles.Linear Color=&Multiply_LinearColorByFloat(
     Linear Color=$Particles.Color
     Float=&Waveform(
      Global Amplitude Scale=5
      "Waveform [1]"=::NewEnumerator0
      "[1] Frequency"=&RandomRangeFloat(Minimum=0.5, Maximum=2)
      "[1] Amplitude Min/Max"=<0 1>
     )
    )
    Particles.Vector 2D:Vector2f=&Multiply_Vector2D(A=$Particles.SpriteSize, B=<0.25 0.25>)
   ) {isSetParameters=true}
   &ScaleColor(Scale Alpha=100, ScaleA=true, ScaleRGB=true) {enabled=false, module=ScaleColor002}
  ]
  rendererClasses [&NiagaraSpriteRendererProperties]
  renderers [{rendererClass=&NiagaraSpriteRendererProperties}]
 }}
summary {
 systemName=NS_FOS_SnowDust
 emitters [{emitterName=Dust_Motes001, simTarget=GPUComputeSim, rendererClasses [&NiagaraSpriteRendererProperties]}]
 userVariables []
}
