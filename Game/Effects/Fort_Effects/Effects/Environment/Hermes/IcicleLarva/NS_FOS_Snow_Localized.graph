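# Niagara system for localized falling snow in the IcicleLarva effects: one GPU sprite emitter of short-lived airborne flakes.
# Flakes spawn in a squashed sphere with a spawn-time kill volume, launched along an angled downward world cone plus initial vortex force.
# Masked, panning curl noise and drag add drift, while sprites rotate, scale with speed and fade in and out.
# A random vector, rotated by particle age, feeds dynamic material parameters as a glint.
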NiagaraSystem /Game/Effects/Fort_Effects/Effects/Environment/Hermes/IcicleLarva/NS_FOS_Snow_Localized
dependencies {
 usedDataInterfaces [&NiagaraDataInterfaceCurve]
 usedDynamicInputs [&Vector2DFromFloat]
 usedModules [&SystemState]
}
emitters {SnowflakesAir {
  simTarget=GPUComputeSim
  emitterUpdateScript [
   &EmitterState()
   &SpawnRate(SpawnRate=40, Use Spawn Probability=false) {module=SpawnRate001}
  ]
  particleSpawnScript [
   &InitializeParticle(
    Lifetime=&UniformRangedFloat(Minimum=3, Maximum=4)
    Sprite Size Mode=::NewEnumerator1
    Uniform Sprite Size Max=6
    UsePositionOffset=false
    Write Material Random=true
   ) {module=InitializeParticle001}
   &ShapeLocation(MaskSpawnGroup=false, Non Uniform Scale=<0.75 0.5 1>, Override Seed=false, Sphere Radius=1200, UseOwnerScale=true)
   &AddVelocityInCone(
    Cone Angle=10
    Cone Axis=<0 -1 -1>
    Cone Axis Coordinate Space=::World
    Module.Use Velocity Falloff On Cone Axis=false
    Velocity Strength=250
   )
   &VortexForce(Origin Pull Amount=100, UseOriginOffset=true, Vortex Axis=<1 0 0>, Vortex Force Amount=100, Vortex Origin Offset:Vector3f=<0 0 -2000>)
   &ApplyInitialForces()
   &KillParticlesInVolume(Sphere Radius=500)
  ]
  particleUpdateScript [
   &ParticleState()
   &ScaleColor(
    Scale Alpha=&FloatFromCurve(
     FloatCurve=DataInterface() {propertyValues {Curve {keys [
         <at -0.0007 -0.0138 RCIM_Cubic>
         {arriveTangent=2.6721, interpMode=RCIM_Cubic, leaveTangent=2.6721, tangentMode=RCTM_User, time=0.3127, value=0.3952}
         {arriveTangent=-0.5712, interpMode=RCIM_Cubic, leaveTangent=-0.5712, tangentMode=RCTM_User, time=0.5424, value=1.0086}
         <at 1.0011 0.002 RCIM_Cubic>
        ]}}}
     Scale Curve=0.5
    )
    ScaleA=true
    ScaleRGB=true
   )
   /Niagara/Modules/Update/Forces/CurlNoiseForce(
    Curl Noise Cone Mask Angle=120
    Curl Noise Cone Mask Falloff Angle=100
    Mask Curl Noise=true
    Noise Frequency=250
    Noise Strength=2000
    Pan Noise=true
    Pan Noise Field=<0 1 2>
   )
   &Drag(Use Linear Drag=true, Use Rotational Drag=true)
   &SolveForcesAndVelocity(Clamp Velocity=false, Limit Acceleration=false)
   &SpriteRotationRate(Rotation Rate=&UniformRangedFloat(Minimum=-15, Maximum=25))
   /SetVariables(Particles.RandomGlintVector:Vector3f=&RotateVectorByQuaternion(
     Vector To Rotate=&RandomVector(Vector Scale=1)
     Quaternion=&MakeQuatFromAxisAngle(
      Axis=&RandomVector(Vector Scale=1)
      AngleInDegrees=&Multiply_Float(A=$Particles.NormalizedAge, B=&RandomRangeFloat(Minimum=2000, Maximum=3000))
     )
    )) {isSetParameters=true}
   &DynamicMaterialParameters(
    Index 0 Param 1=&RandomRangeFloat(Minimum=0, Maximum=1)
    Index 0 Param 2=&MakeFloatFromVector(Vector=$Particles.RandomGlintVector, Channel=::NewEnumerator0)
    Index 0 Param 3=&MakeFloatFromVector(Vector=$Particles.RandomGlintVector, Channel=::NewEnumerator1)
    Index 0 Param 4=&MakeFloatFromVector(Vector=$Particles.RandomGlintVector, Channel=::NewEnumerator2)
    Module.Param0WriteEnabled=true
    Module.Param1WriteEnabled=true
    Module.Param2WriteEnabled=true
    Module.Param3WriteEnabled=true
   )
   &ScaleSpriteSizeBySpeed(Max Scale Factor=<3 5>)
   &ScaleSpriteSize(Scale Sprite Size Mode=::NewEnumerator0, Uniform Scale Factor=5) {enabled=false}
   &ScaleColor(Scale RGB=<100 1 1>, ScaleA=true, ScaleRGB=true) {enabled=false, module=ScaleColor001}
  ]
  rendererClasses [&NiagaraSpriteRendererProperties]
  renderers [{rendererClass=&NiagaraSpriteRendererProperties}]
 }}
summary {
 systemName=NS_FOS_Snow_Localized
 emitters [{emitterName=SnowflakesAir, simTarget=GPUComputeSim, rendererClasses [&NiagaraSpriteRendererProperties]}]
 userVariables []
}
