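# Niagara snowfall system, named as player-attached, that fills the surrounding area with short-lived, drifting snowflake sprites.
# A GPU air emitter adds curl noise and vortex forces, with a lower-rate CPU variant for mobile.
# A GPU ground emitter places flakes just above the terrain using FN_SpawnParticlesOnTerrainNearPlayer.
# Flakes fade and stretch with speed; air flakes feed spinning glint vectors to materials, and time-of-day spawn scaling is disabled.
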NiagaraSystem /Game/Effects/Fort_Effects/Effects/Environment/Hermes/IcicleLarva/NS_FOS_Snow_PlayerAttached
dependencies {
 usedDataInterfaces [&NiagaraDataInterfaceCurve, &NiagaraDataInterfaceLandscape]
 usedDynamicInputs []
 usedModules [&SystemState]
}
emitters {
 SnowflakesAir {
  simTarget=GPUComputeSim
  emitterUpdateScript [
   &EmitterState()
   &SpawnRate(SpawnRate=2500, Use Spawn Probability=false) {module=SpawnRate001}
   &SpawnRate(
    SpawnRate=&Multiply_Float(
     A=2500
     B=&FloatFromCurve(
      FloatCurve=DataInterface() {propertyValues {Curve {keys [
          {interpMode=RCIM_Cubic, tangentMode=RCTM_User, time=0, value=1}
          <at 0.1 0.2>
          {interpMode=RCIM_Cubic, tangentMode=RCTM_User, time=0.25, value=0.25}
          {interpMode=RCIM_Cubic, tangentMode=RCTM_User, time=0.4199, value=0.9966}
          {interpMode=RCIM_Cubic, tangentMode=RCTM_User, time=0.5, value=1}
          {interpMode=RCIM_Cubic, tangentMode=RCTM_User, time=0.5967, value=1.0046}
          <at 0.7666 0.25>
          {interpMode=RCIM_Cubic, tangentMode=RCTM_User, time=0.95, value=0.25}
          {interpMode=RCIM_Cubic, tangentMode=RCTM_User, time=1, value=1}
         ]}}}
      CurveIndex=Linked() {linkedVariable {name=NPC.FortniteNPC.TimeOfDay_0-1, type=NiagaraFloat}}
     )
    )
    Use Spawn Probability=false
   ) {enabled=false}
  ]
  particleSpawnScript [
   /Niagara/Modules/Spawn/Initialization/InitializeParticle(
    Lifetime=&UniformRangedFloat(Minimum=2, Maximum=3)
    Module.Write Color=true
    Module.Write Lifetime=true
    Module.Write Mass=true
    Module.Write Material Random=true
    Module.Write Position=false
    Module.Write Scale=false
    Module.Write SpriteRotation=true
    Module.Write SpriteSize=true
    Sprite Rotation=&UniformRangedFloat(Minimum=0, Maximum=360)
    Sprite Size=&Vector2DFromFloat(Value=&UniformRangedFloat(Minimum=6, Maximum=7))
   )
   /Niagara/Modules/Spawn/Location/SphereLocation(MaskSpawnGroup=false, Module.Override Randomness=false, Sphere Radius=8000, Surface Only=false, Surface Only Band Thickness=1)
   &CylinderLocation(
    Cylinder Radius=2000
    Lathe Profile=DataInterface() {propertyValues {Curve {keys []}}}
    MaskSpawnGroup=false
    Module.Enable Lathe Profile=false
    Module.Override Randomness=false
    Module.Surface Only=true
    Use Endcaps In Surface Only Mode=false
   )
   &AddVelocityInCone(Cone Angle=1, Cone Axis=<1 0 -1>, Cone Axis Coordinate Space=::World, Module.Use Velocity Falloff On Cone Axis=false, Velocity Strength=150)
   /Niagara/Modules/Update/Forces/CurlNoiseForce(
    Curl Noise Cone Mask Angle=120
    Curl Noise Cone Mask Falloff Angle=100
    Mask Curl Noise=true
    Noise Frequency=250
    Noise Strength=800
    Pan Noise=true
    Pan Noise Field=<0 1 2>
   )
   &VortexForce(Origin Pull Amount=300, UseOriginOffset=true, Vortex Axis=<1 0 0>, Vortex Force Amount=300, Vortex Origin Offset:Vector3f=<0 0 -5000>)
   &ApplyInitialForces()
   &KillParticlesInVolume(Sphere Radius=4500)
  ]
  particleUpdateScript [
   &ParticleState() {module=UpdateAge}
   &ScaleColor(
    Scale Alpha=&FloatFromCurve(
     FloatCurve=DataInterface() {propertyValues {Curve {keys [
         <at -0.0007 -0.0138 RCIM_Cubic>
         {arriveTangent=2.6721, interpMode=RCIM_Cubic, leaveTangent=2.6721, tangentMode=RCTM_User, time=0.3127, value=0.3952}
         {arriveTangent=-0.5712, interpMode=RCIM_Cubic, leaveTangent=-0.5712, tangentMode=RCTM_User, time=0.5424, value=1.0086}
         <at 1.0011 0.002 RCIM_Cubic>
        ]}}}
     Scale Curve=0.5
    )
    ScaleA=true
    ScaleRGB=true
   )
   &SolveForcesAndVelocity(Clamp Velocity=false, Limit Acceleration=false)
   &SpriteRotationRate(Rotation Rate=&UniformRangedFloat(Minimum=-15, Maximum=25))
   /SetVariables(Particles.RandomGlintVector:Vector3f=&RotateVectorByQuaternion(
     Vector To Rotate=&RandomVector(Vector Scale=1)
     Quaternion=&MakeQuatFromAxisAngle(
      Axis=&RandomVector(Vector Scale=1)
      AngleInDegrees=&Multiply_Float(A=$Particles.NormalizedAge, B=&RandomRangeFloat(Minimum=2000, Maximum=3000))
     )
    )) {isSetParameters=true, module=SetVariables_#1}
   &DynamicMaterialParameters(
    Index 0 Param 1=&RandomRangeFloat(Minimum=0, Maximum=1)
    Index 0 Param 2=&MakeFloatFromVector(Vector=$Particles.RandomGlintVector, Channel=::NewEnumerator0)
    Index 0 Param 3=&MakeFloatFromVector(Vector=$Particles.RandomGlintVector, Channel=::NewEnumerator1)
    Index 0 Param 4=&MakeFloatFromVector(Vector=$Particles.RandomGlintVector, Channel=::NewEnumerator2)
    Module.Param0WriteEnabled=true
    Module.Param1WriteEnabled=true
    Module.Param2WriteEnabled=true
    Module.Param3WriteEnabled=true
   )
   &ScaleSpriteSizeBySpeed(Max Scale Factor=<3 5>)
   &ScaleColor(Scale RGB=<100 1 1>, ScaleA=true, ScaleRGB=true) {enabled=false, module=ScaleColor001}
  ]
  rendererClasses [&NiagaraSpriteRendererProperties]
  renderers [{rendererClass=&NiagaraSpriteRendererProperties}]
 }
 SnowflakesAir_Mobile {
  emitterUpdateScript [
   &EmitterState()
   &SpawnRate(SpawnRate=250, Use Spawn Probability=false)
  ]
  particleSpawnScript [
   &InitializeParticle(
    Color=<rgba 1 1 1 0.75>
    Lifetime=&UniformRangedFloat(Minimum=2, Maximum=3)
    Module.Write Color=true
    Module.Write Lifetime=true
    Module.Write Mass=true
    Module.Write Material Random=true
    Module.Write Position=false
    Module.Write Scale=false
    Module.Write SpriteRotation=true
    Module.Write SpriteSize=true
    Sprite Rotation=&UniformRangedFloat(Minimum=0, Maximum=360)
    Sprite Size=&Vector2DFromFloat(Value=&UniformRangedFloat(Minimum=9, Maximum=12))
   )
   &SphereLocation(MaskSpawnGroup=false, Module.Override Randomness=false, Sphere Radius=8000, Surface Only=false, Surface Only Band Thickness=1)
   &CylinderLocation(Cylinder Radius=2000, MaskSpawnGroup=false, Override Seed=false, UseOwnerScale=true)
   &AddVelocity(Velocity=&UniformRangedVector(Minimum=<0 -300 -100>, Maximum=<0 -600 -200>))
   &CurlNoiseForce(Noise Frequency=100, Noise Strength=1500, Pan Noise=false, Pan Noise Field=<0 1 0>)
   &ApplyInitialForces()
   &AddVelocityInCone(Cone Angle=10, Cone Axis Coordinate Space=::World, Module.Use Velocity Falloff On Cone Axis=false, Velocity Strength=250)
   &KillParticlesInVolume(Sphere Radius=4500)
  ]
  particleUpdateScript [
   &ParticleState() {module=UpdateAge}
   &ScaleColor(
    Scale Alpha=&FloatFromCurve(FloatCurve=DataInterface() {propertyValues {Curve {keys [
         <at -0.0007 -0.0138 RCIM_Cubic>
         {arriveTangent=2.6721, interpMode=RCIM_Cubic, leaveTangent=2.6721, tangentMode=RCTM_User, time=0.3127, value=0.3952}
         {arriveTangent=-0.5712, interpMode=RCIM_Cubic, leaveTangent=-0.5712, tangentMode=RCTM_User, time=0.5424, value=1.0086}
         <at 1.0011 0.002 RCIM_Cubic>
        ]}}})
    ScaleA=true
    ScaleRGB=true
   )
   &SolveForcesAndVelocity(Clamp Velocity=false, Limit Acceleration=false)
   &SpriteRotationRate(Rotation Rate=&UniformRangedFloat(Minimum=-15, Maximum=25))
   /SetVariables(Particles.RandomGlintVector:Vector3f=&RotateVectorByQuaternion(
     Vector To Rotate=&RandomVector(Vector Scale=1)
     Quaternion=&MakeQuatFromAxisAngle(
      Axis=&RandomVector(Vector Scale=1)
      AngleInDegrees=&Multiply_Float(A=$Particles.NormalizedAge, B=&RandomRangeFloat(Minimum=2000, Maximum=3000))
     )
    )) {isSetParameters=true, module=SetVariables_#1}
   &DynamicMaterialParameters(
    Index 0 Param 1=&RandomRangeFloat(Minimum=0, Maximum=1)
    Index 0 Param 2=&MakeFloatFromVector(Vector=$Particles.RandomGlintVector, Channel=::NewEnumerator0)
    Index 0 Param 3=&MakeFloatFromVector(Vector=$Particles.RandomGlintVector, Channel=::NewEnumerator1)
    Index 0 Param 4=&MakeFloatFromVector(Vector=$Particles.RandomGlintVector, Channel=::NewEnumerator2)
    Module.Param0WriteEnabled=true
    Module.Param1WriteEnabled=true
    Module.Param2WriteEnabled=true
    Module.Param3WriteEnabled=true
   )
   &ScaleSpriteSizeBySpeed(Max Scale Factor=<2 5>)
  ]
  rendererClasses [&NiagaraSpriteRendererProperties]
  renderers [{rendererClass=&NiagaraSpriteRendererProperties}]
 }
 SnowflakesGround {
  simTarget=GPUComputeSim
  emitterUpdateScript [
   &EmitterState()
   &SpawnRate(SpawnRate=2500, Use Spawn Probability=false) {module=SpawnRate001}
   &SpawnRate(
    SpawnRate=&Multiply_Float(
     A=2500
     B=&FloatFromCurve(
      FloatCurve=DataInterface() {propertyValues {Curve {keys [
          {interpMode=RCIM_Cubic, tangentMode=RCTM_User, time=0, value=1}
          <at 0.1 0.25>
          {interpMode=RCIM_Cubic, tangentMode=RCTM_User, time=0.25, value=0.25}
          {interpMode=RCIM_Cubic, tangentMode=RCTM_User, time=0.4199, value=0.9966}
          {interpMode=RCIM_Cubic, tangentMode=RCTM_User, time=0.5, value=1}
          {interpMode=RCIM_Cubic, tangentMode=RCTM_User, time=0.5967, value=1.0046}
          <at 0.7666 0.25>
          {interpMode=RCIM_Cubic, tangentMode=RCTM_User, time=0.95, value=0.25}
          {interpMode=RCIM_Cubic, tangentMode=RCTM_User, time=1, value=1}
         ]}}}
      CurveIndex=Linked() {linkedVariable {name=NPC.FortniteNPC.TimeOfDay_0-1, type=NiagaraFloat}}
     )
    )
    Use Spawn Probability=false
   ) {enabled=false}
  ]
  particleSpawnScript [
   &InitializeParticle(
    Lifetime=&UniformRangedFloat(Minimum=2, Maximum=3)
    Module.Write Color=true
    Module.Write Lifetime=true
    Module.Write Mass=true
    Module.Write Material Random=true
    Module.Write Position=false
    Module.Write Scale=false
    Module.Write SpriteRotation=true
    Module.Write SpriteSize=true
    Sprite Rotation=&UniformRangedFloat(Minimum=0, Maximum=360)
    Sprite Size=&Vector2DFromFloat(Value=&UniformRangedFloat(Minimum=5, Maximum=7))
   ) {enabled=false}
   /Niagara/Modules/Spawn/Initialization/V2/InitializeParticle(
    Lifetime Max=3
    Lifetime Min=2
    Lifetime Mode=::NewEnumerator1
    Sprite Size Mode=::NewEnumerator1
    UsePositionOffset=false
    Write Material Random=true
   ) {module=InitializeParticle001}
   &SphereLocation(
    MaskSpawnGroup=false
    Module.Override Randomness=false
    Non Uniform Scale=<1 1 0>
    Sphere Radius=8000
    Surface Only=false
    Surface Only Band Thickness=1
   )
   &CylinderLocation(
    Cylinder Radius=3000
    Lathe Profile=DataInterface() {propertyValues {Curve {keys []}}}
    MaskSpawnGroup=false
    Module.Enable Lathe Profile=false
    Module.Override Randomness=false
    Module.Surface Only=true
    Use Endcaps In Surface Only Mode=false
   )
   /NiagaraFN/Modules/Spawn/FN_SpawnParticlesOnTerrainNearPlayer(
    Height Texture -x -y:Vector3f=<-130048 -130048 0>
    HeightOffset=50
    HeightTexture +x +y:Vector3f=<130048 130048 0>
    Player Position:NiagaraPosition=$Particles.Position
    Player Relative Particle Location:Vector2f=&UniformRangedVector2D(Minimum=<0 0>, Maximum=<0 0>)
   ) {module=FN_SpawnParticlesOnTerrainNearPlayer001}
   &AddVelocity(Velocity=&UniformRangedVector(Minimum=<600 -256 100>, Maximum=<1200 256 150>)) {enabled=false}
   &CurlNoiseForce(Noise Frequency=25, Noise Strength=300, Pan Noise=true)
   &ApplyInitialForces()
   &AddVelocityInCone(Cone Angle=15, Cone Axis Coordinate Space=::World, Module.Use Velocity Falloff On Cone Axis=false, Velocity Strength=250) {module=AddVelocityInCone001}
   &KillParticlesInVolume(Sphere Radius=4500)
  ]
  particleUpdateScript [
   &ParticleState() {module=UpdateAge}
   &ScaleColor(
    Scale Alpha=&FloatFromCurve(FloatCurve=DataInterface() {propertyValues {Curve {keys [
         <at -0.0007 -0.0138 RCIM_Cubic>
         {arriveTangent=2.6721, interpMode=RCIM_Cubic, leaveTangent=2.6721, tangentMode=RCTM_User, time=0.3127, value=0.3952}
         {arriveTangent=-0.5712, interpMode=RCIM_Cubic, leaveTangent=-0.5712, tangentMode=RCTM_User, time=0.5424, value=1.0086}
         <at 1.0011 0.002 RCIM_Cubic>
        ]}}})
    ScaleA=true
    ScaleRGB=true
   )
   &SolveForcesAndVelocity(Clamp Velocity=false, Limit Acceleration=false)
   &SpriteRotationRate(Rotation Rate=&UniformRangedFloat(Minimum=-15, Maximum=25))
   &ScaleSpriteSizeBySpeed(Max Scale Factor=<2 5>)
   &ScaleColor(Scale RGB=<100 1 1>, ScaleA=true, ScaleRGB=true) {enabled=false, module=ScaleColor001}
  ]
  rendererClasses [&NiagaraSpriteRendererProperties]
  renderers [{rendererClass=&NiagaraSpriteRendererProperties}]
 }
}
summary {
 systemName=NS_FOS_Snow_PlayerAttached
 emitters [
  {emitterName=SnowflakesGround, simTarget=GPUComputeSim, rendererClasses [&NiagaraSpriteRendererProperties]}
  {emitterName=SnowflakesAir, simTarget=GPUComputeSim, rendererClasses [&NiagaraSpriteRendererProperties]}
  {emitterName=SnowflakesAir_Mobile, rendererClasses [&NiagaraSpriteRendererProperties]}
 ]
 userVariables []
}
